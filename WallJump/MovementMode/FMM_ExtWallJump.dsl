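; Wall jump movement mode extension that drives a looping gameplay cue on the owning pawn while active.
; On activation it updates the surface type and classifies the runtime jump direction via dot products as backward, rightward or other, passed as the cue's normalized magnitude.
; The cue normal points from the wall hit toward the pawn, and the direction value is also debug-printed.
; Deactivation removes the cue, which is held in a variable named as temporary.
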(event EventOnExtensionActivate
  (params (RuntimeContext :type MMERuntime%20Context%20Structure%20%28by%20ref%29))
  (bind _ownerpawn (Variables|FortMovementMode_BaseExtLogic|GetOwnerPawn))
  (bind _returnvalue (Transformation|GetActorRightVector _ownerpawn))
  (bind _ownerpawn_2 (Variables|FortMovementMode_BaseExtLogic|GetOwnerPawn))
  (bind _returnvalue_2 (Transformation|GetActorForwardVector _ownerpawn_2))
  (|CallParentFunction RuntimeContext)
  (CallFunction|UpdateSurfaceType)
  (bind _returnvalue_3 (Class|FortMovementModeExtBlueprintLibrary|GetRuntimeDatafromContext RuntimeContext "/Script/WallJumpCodeRuntime.FortMovementMode_WallJumpRuntimeData"))
  (Development|PrintString 0 (Utilities|String|ToString(Float) (Math|Float|SelectFloat -1.0 (Math|Float|SelectFloat 0.0 1.0 (> (Math|Vector|DotProduct _returnvalue (Variables|/Script/WallJumpCodeRuntime.FortMovementMode_WallJumpRuntimeData|GetNetJumpDirection _returnvalue_3)) 0)) (< (Math|Vector|DotProduct (Variables|/Script/WallJumpCodeRuntime.FortMovementMode_WallJumpRuntimeData|GetNetJumpDirection _returnvalue_3) _returnvalue_2) -0.8))))
  (GameplayCue|AddGameplayCueOnActor(Looping) (Variables|FortMovementMode_BaseExtLogic|GetOwnerPawn) (Variables|Default|GetTempWallJumpCue) (Ability|GameplayCue|MakeGameplayCueParameters :NormalizedMagnitude (Math|Float|SelectFloat -1.0 (Math|Float|SelectFloat 0.0 1.0 (> (Math|Vector|DotProduct _returnvalue (Variables|/Script/WallJumpCodeRuntime.FortMovementMode_WallJumpRuntimeData|GetNetJumpDirection _returnvalue_3)) 0)) (< (Math|Vector|DotProduct (Variables|/Script/WallJumpCodeRuntime.FortMovementMode_WallJumpRuntimeData|GetNetJumpDirection _returnvalue_3) _returnvalue_2) -0.8)) :RawMagnitude 0.0 :Location "0,0,0" :Normal (Math|Vector|GetUnitDirection(Vector) (.Location (Collision|BreakHitResult (Variables|/Script/WallJumpCodeRuntime.FortMovementMode_WallJumpRuntimeData|GetWallHitResult _returnvalue_3))) (Transformation|GetActorLocation (Variables|FortMovementMode_BaseExtLogic|GetOwnerPawn))) :GameplayEffectLevel 1 :AbilityLevel 1 :bReplicateLocationWhenUsingMinimalRepProxy false)))

(event EventOnExtensionDeactivate
  (params (RuntimeContext :type MMERuntime%20Context%20Structure%20%28by%20ref%29))
  (|CallParentFunction RuntimeContext)
  (GameplayCue|RemoveGameplayCueOnActor(Looping) (Variables|FortMovementMode_BaseExtLogic|GetOwnerPawn) (Variables|Default|GetTempWallJumpCue) (Ability|GameplayCue|MakeGameplayCueParameters 0.0)))
